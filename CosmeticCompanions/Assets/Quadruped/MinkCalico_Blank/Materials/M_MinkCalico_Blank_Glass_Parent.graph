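# Translucent lit parent material for the glass helmet of the Mink Calico quadruped cosmetic companion, with an animatable reveal.
# A fresnel glass base with noise normals gains optional texture masking, thin film, camera dither and vertex-offset effects.
# Helmet Reveal drives a noise-distorted local-Z mask that cuts opacity and forms a glowing edge band.
# Band, fresnel and masked depth-fade drive scanline emissive, with opacity passed through translucency sorting before tech-art overrides.
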Material /CosmeticCompanions/Assets/Quadruped/MinkCalico_Blank/Materials/M_MinkCalico_Blank_Glass_Parent
graph {
 properties {
  allowTranslucentCustomDepthWrites=true
  bAutomaticallySetUsageInEditor=false
  bEnableResponsiveAA=true
  bScreenSpaceReflections=true
  bUseMaterialAttributes=true
  bUsedWithInstancedStaticMeshes=true
  bUsedWithMeshDeformer=true
  bUsedWithMorphTargets=true
  bUsedWithSkeletalMesh=true
  blendMode=BLEND_Translucent
  ditherOpacityMask=true
  pixelDepthOffsetMode=PDOM_Legacy
  translucencyDirectionalLightingIntensity=5
  translucencyLightingMode=TLM_SurfacePerPixelLighting
  usedShadingModels=Default Lit
 }
 nodes [
  MaterialFunctionCall!(WorldPostionOfBlackHole, World Position) {materialFunction=/Game/Packages/Fortnite_UniqueMaterials_1/BlackHoleFunction}
  n1 = Constant3Vector()
  n2 = Multiply(A=SkyAtmosphereLightDirection(), B=ScalarParameter() {defaultValue=-1, group=Glass, parameterName=ShadowBufferOffset}) {constB=-1}
  n3 = ScalarParameter() {parameterName=ThinFilmOpacity}
  n4 = TextureSampleParameter2D(UVs=TextureCoordinate()) {group=Mask Texture, parameterName=Diffuse Texture with Alpha Mask, sortPriority=10, texture=&T_BlacksRGB}
  n5 = MaterialFunctionCall(Float3=TextureSampleParameter2D() {parameterName=Input S, texture=/Game/Global/Textures/Default/Blanks/T_BlackUserInterface}.RGB) {materialFunction=&BreakOutFloat3Components}
  n6 = StaticBoolParameter() {group=Specular, parameterName=SwizzleRoughnessToGreen}
  n7 = MaterialFunctionCall(Float3=MaterialFunctionCall(X=@n5.R, Y=StaticSwitch(True=@n5.B, False=@n5.B, Value=@n6), Z=StaticSwitch(True=@n5.G, False=@n5.G, Value=@n6)) {materialFunction=&MakeFloat3}.Result) {materialFunction=&BreakOutFloat3Components}
  n8 = ScalarParameter() {defaultValue=1, parameterName=Fresnel Exponent}
  n9 = ObjectBounds()
  n10 = PreSkinnedLocalBounds()
  n11 = Saturate(None=Floor(None=ScalarParameter() {parameterName=Use Pre-Skinned Bounds, sliderMax=1}))
  n12 = ComponentMask(None=Divide(A=PreSkinnedPosition(), B=LinearInterpolate(A=@n9, B=@n10.Half Extents, Alpha=@n11))) {b=true, g=true}
  n13 = ScalarParameter() {defaultValue=0.2, parameterName=WaterTiling}
  n14 = TextureSampleParameter2D(UVs=VertexInterpolator(VS=Multiply(A=@n12, B=@n13)).PS) {
   parameterName=Noise Normal
   samplerSource=SSM_Wrap_WorldGroupSettings
   samplerType=SAMPLERTYPE_Normal
   texture=/Game/Global/Textures/Default/Noise/T_Noise_Normal_01
  }
  n15 = VectorParameter() {parameterName=Normal Intensity}
  n16 = TextureSampleParameter2D() {
   group=Mask Texture
   parameterName=Normals
   samplerSource=SSM_Wrap_WorldGroupSettings
   samplerType=SAMPLERTYPE_Normal
   texture=/Engine/EngineMaterials/FlatNormal
  }
  n17 = Normalize(VectorInput=Add(A=Multiply(A=@n14.RGB, B=@n15.RGB), B=@n16.RGB))
  n18 = Transform(None=@n17)
  n19 = Clamp(None=Multiply(A=Fresnel(ExponentIn=@n8, Normal=@n18), B=ScalarParameter() {defaultValue=1, parameterName=Glass thickness}))
  n20 = LinearInterpolate(
   A=ScalarParameter() {parameterName=Fresnel Inner Emissive}
   B=ScalarParameter() {defaultValue=0.85, parameterName=Fresnel Outer Emissive}
   Alpha=@n19
  )
  n21 = VectorParameter() {parameterName=EmissiveColor}
  n22 = TextureSampleParameter2D() {parameterName=Input FX, texture=&T_BlackUserInterface}
  n23 = ScalarParameter() {parameterName=Add Custom Opacity, sortPriority=10}
  n24 = ScalarParameter() {parameterName=Fresnel Inner Transparency}
  n25 = ScalarParameter() {defaultValue=0.85, parameterName=Fresnel Outer Transparency}
  n26 = LinearInterpolate(
   A=Add(A=Multiply(A=@n22.B, B=@n23), B=LinearInterpolate(A=@n24, B=@n25, Alpha=@n19))
   B=@n22.B
   Alpha=ScalarParameter() {parameterName=Use Custom Opacity}
  )
  n27 = MakeMaterialAttributes!(
   BaseColor=StaticSwitchParameter(True=@n4.RGB, False=VectorParameter() {parameterName=Base Color}.RGB) {parameterName="Use Diffuse Texture for Color [ignores alpha channel]"}
   Metallic=LinearInterpolate(A=ScalarParameter() {defaultValue=0.1, parameterName=Metallic}, B=@n7.G, Alpha=ScalarParameter() {parameterName=Use Custom Metallic})
   Specular=LinearInterpolate(A=ScalarParameter() {defaultValue=1, parameterName=Specular}, B=@n7.R, Alpha=ScalarParameter() {parameterName=Use Custom Specular})
   Roughness=LinearInterpolate(A=ScalarParameter() {parameterName=Roughness}, B=@n7.B, Alpha=ScalarParameter() {parameterName=Use Custom Roughness})
   Anisotropy
   EmissiveColor=Multiply(
    A=Multiply(A=@n20, B=@n21.RGB)
    B=TextureSampleParameter2D() {parameterName=EmissiveMask, samplerType=SAMPLERTYPE_Masks, texture=&T_BlackMask}.RGB
   )
   Opacity=@n26
   OpacityMask=@n26
   Normal=@n17
   Tangent
   WorldPositionOffset
   SubsurfaceColor
   ClearCoat
   ClearCoatRoughness
   AmbientOcclusion
   Refraction
   CustomizedUVs_0
   CustomizedUVs_1
   CustomizedUVs_2
   CustomizedUVs_3
   CustomizedUVs_4
   CustomizedUVs_5
   CustomizedUVs_6
   CustomizedUVs_7
   PixelDepthOffset
   ShadingModel
   Displacement
  )
  n28 = GetMaterialAttributes(None=@n27) {attributeGetTypes [
    #1
    #2
    #3
    #4
    #5
   ]}
  n29 = Multiply(A=@n4.A, B=ScalarParameter() {defaultValue=1, group=Mask Texture, parameterName=Alpha Channel Mask Opacity})
  n30 = ScalarParameter() {defaultValue=0.4, group=Mask Texture, parameterName=Masked Area Roughness, sortPriority=4}
  n31 = Saturate(None=LinearInterpolate(A=@n28.Opacity, Alpha=@n29))
  n32 = SetMaterialAttributes!(
   MaterialAttributes=@n28.MaterialAttributes
   Base Color=Saturate(None=LinearInterpolate(A=@n28.Base Color, B=@n4.RGB, Alpha=@n29))
   Metallic=LinearInterpolate(A=@n28.Metallic, B=ScalarParameter() {group=Mask Texture, parameterName=Masked Area Metallic, sortPriority=2}, Alpha=@n29)
   Specular=LinearInterpolate(A=@n28.Specular, B=ScalarParameter() {defaultValue=0.5, group=Mask Texture, parameterName=Masked Area Specular, sortPriority=3}, Alpha=@n29)
   Roughness=Saturate(None=LinearInterpolate(A=@n28.Roughness, B=@n30, Alpha=@n29))
   Opacity=@n31
   Opacity Mask=@n31
  ) {attributeSetTypes [
    #1
    #2
    #3
    #4
    #5
    #6
   ]}
  n33 = StaticSwitchParameter(True=@n32, False=@n27) {group=Mask Texture, parameterName=Use Diffuse Texture with Alpha Mask, sortPriority=-50}
  n34 = Multiply(A=Subtract(A=ObjectPositionWS(), B=WorldPosition().XYZ), B=ScalarParameter() {parameterName=HT_CrunchVerts})
  n35 = VectorParameter() {parameterName=WorldPostionOfBlackHole, defaultValue=<rgba -30000 25536 1952 0>}
  n36 = MaterialFunctionCall!(
   World Position=WorldPosition().XYZ
   DistanceCullFadeAnimationOffset=MaterialFunctionCall() {materialFunction=/Game/Packages/Fortress_SharedMaterials/Base_Material_Functions/PickupEffectDistanceCullOverride}.Result
   Use actor center instead of pivot point
  ) {materialFunction=/Game/Packages/Fortress_SharedMaterials/Base_Material_Functions/LODEffect}
  n37 = WorldPosition()
  n38 = Add(
   A=MaterialFunctionCall(WorldPostionOfBlackHole=AppendVector(A=@n35.RGB, B=@n35.RGB), World Position=Add(A=@n36.Result, B=@n37.XYZ)) {materialFunction=&BlackHoleFunction}.Result
   B=@n36.Result
  )
  n39 = SetMaterialAttributes!(
   MaterialAttributes=StaticSwitchParameter(
    True=MaterialFunctionCall!(Alpha=@n3, Material Attributes=@n33) {materialFunction=/Game/Weapons/WeaponSkins/MaterialFunction/ThinFilmFunction}.Result
    False=@n33
   ) {parameterName=useThinFilm}
   World Position Offset=StaticSwitchParameter(True=Add(A=@n34, B=@n38), False=@n38) {parameterName=bUseVertexHide}
  ) {attributeSetTypes [#7]}
  n40 = StaticSwitchParameter(
   True=MaterialFunctionCall!(In=@n39) {materialFunction=/Game/Packages/Fortress_SharedMaterials/Base_Material_Functions/BaseCustomizationFunctions/MF_FN_Vehicle_CameraDither}.Result
   False=@n39
  ) {parameterName=UseCameraDither}
  n41 = StaticSwitchParameter(
   True=SetMaterialAttributes(MaterialAttributes=@n40, World Position Offset=ShadowReplace(Default=@n1, Shadow=@n2)) {attributeSetTypes [#7]}
   False=@n40
  ) {desc=Fix Glass Artifacts When Cast Dynamic Shadow as Mask is ON, group=Glass, parameterName=Use_GlassCastShadowOffsetFix, sortPriority=0}
  n42 = GetMaterialAttributes(None=@n41) {attributeGetTypes [#5, #1]}
  n43 = OneMinus(None=ScalarParameter() {group=Mink Calico - Glass, parameterName=Helmet Reveal})
  n44 = LocalPosition() {localOrigin=InstancePreSkinning}
  n45 = MaterialFunctionCall!(
   Tile=ScalarParameter() {defaultValue=3, group=Mink Calico - Glass, parameterName=Distortion - Tiling}
   DistortionAdjustment=Constant() {r=-5}
  ) {materialFunction=/Game/Global/Materials/MaterialGym/Materials/MaterialElements/UVs/CylindricalProjection/MF_CylindricalProjUV_Mod}
  n46 = Add(
   A=Divide(A=VertexInterpolator(VS=@n44.Z).PS, B=ScalarParameter() {defaultValue=200, parameterName=Reveal Mask - Distance})
   B=Multiply(
    A=TextureSample(UVs=@n45.UVs) {samplerType=SAMPLERTYPE_Grayscale, texture=&LowResBlurredNoise}.R
    B=ScalarParameter() {defaultValue=0.0477, group=Mink Calico - Glass, parameterName="Distortion -  Intensity"}
   )
  )
  n47 = SmoothStep(
   Min=@n43
   Max=Add(A=@n43, B=ScalarParameter() {defaultValue=0.0256, group=Mink Calico - Glass, parameterName=Reveal Mask - Smoothness})
   Value=@n46
  )
  n48 = ScalarParameter() {defaultValue=0.5606, parameterName=Helmet Band - Thickness}
  n49 = ScalarParameter() {defaultValue=-0.4205, parameterName=Helmet Band - Offset}
  n50 = SmoothStep(
   Min=@n48
   Max=Add(A=@n48, B=ScalarParameter() {defaultValue=0.3776, parameterName=Helmet Band - Smoothness})
   Value=Sine(None=Add(A=@n47, B=@n49))
  )
  n51 = Max(A=Multiply(A=@n42.Opacity, B=OneMinus(None=@n47)), B=@n50)
  n52 = VectorParameter() {group=Mink Calico - Glass, parameterName=Emissive Color, defaultValue=<rgba 0 3.2673 5 1>}
  n53 = MaterialFunctionCall!(
   Number of Scanlines=ScalarParameter() {defaultValue=55, group=Mink Calico - Glass, parameterName=Scanline - Size}
   Scanline Speed=ScalarParameter() {defaultValue=0.0198, group=Mink Calico - Glass, parameterName=Scanline - Speed}
   "Scanlines Thickness (Bias)"
   Scanlines Dark Value
   Scanlines Light Value=ScalarParameter() {defaultValue=1, group=Mink Calico - Glass, parameterName=Scanline - Light Value}
   Block Opacity
   UVs=MaterialFunctionCall!(Position, UVRadialScale) {materialFunction=/Game/Weapons/WeaponSkins/MaterialFunction/ObjectStabilizedScreenSpaceUVs}.StabilizedUVs
   Time
   Block Repeat
  ) {materialFunction=/Game/Effects/Fort_Effects/Effects/Atlas/Materials/MF_Atlas_Scanlines_01}
  n54 = ScalarParameter() {defaultValue=5, group=Mink Calico - Glass, parameterName=Scanline - Intensity}
  n55 = ScalarParameter() {defaultValue=0.6012, group=Mink Calico - Glass, parameterName=Fresnel - Threshold}
  n56 = Add(A=@n55, B=ScalarParameter() {defaultValue=0.096, group=Mink Calico - Glass, parameterName=Fresnel - Smoothness})
  n57 = Fresnel(ExponentIn=ScalarParameter() {defaultValue=2.5502, group=Mink Calico - Glass, parameterName=Fresnel - Exp}) {baseReflectFraction=0}
  n58 = ScalarParameter() {defaultValue=2, group=Mink Calico - Glass, parameterName=Fresnel - Intensity}
  n59 = Multiply(
   A=Max(A=SmoothStep(Min=@n55, Max=@n56, Value=@n46), B=Multiply(A=@n57, B=@n58))
   B=ScalarParameter() {defaultValue=0.25, group=Mink Calico - Glass, parameterName="Fresnel  - Intensity"}
  )
  n60 = ScalarParameter() {defaultValue=1.3363, group=Mink Calico - Glass, parameterName=Depth Fade - Distance}
  n61 = Multiply(
   A=Sine(None=DepthFade(FadeDistance=@n60))
   B=ScalarParameter() {defaultValue=8.5772, group=Mink Calico - Glass, parameterName=Depth Fade - Intensity}
  )
  n62 = TextureSampleParameter2D() {parameterName=Input FX, texture=&T_BlackUserInterface}
  n63 = LinearInterpolate(B=Multiply(A=@n52.RGB, B=Multiply(A=@n53.Result, B=@n54)), Alpha=Max(A=Max(A=@n50, B=@n59), B=Multiply(A=@n61, B=@n62.R)) {constB=0})
  n64 = Multiply(A=@n52.RGB) {constB=3}
  n65 = Saturate(None=@n50)
  MaterialFunctionCall(In=SetMaterialAttributes!(
    MaterialAttributes=@n42.MaterialAttributes
    Opacity=MaterialFunctionCall!(Opacity=@n51) {materialFunction=/Game/Global/Materials/MaterialGym/Materials/SortedOutline/MF_SortTranslucency}.Result
    Emissive Color=LinearInterpolate(A=@n63, B=@n64, Alpha=@n50)
    Base Color=LinearInterpolate(A=@n42.Base Color, Alpha=@n65) {constB=0}
   ) {attributeSetTypes [#5, #8, #1]}) {materialFunction=&MF_TechArtOverrides}
 ]
}
